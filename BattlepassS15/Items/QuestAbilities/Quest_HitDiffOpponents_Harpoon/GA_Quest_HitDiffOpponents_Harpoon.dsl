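; Quest gameplay ability that tracks hits on different opponents with the harpoon gun for a Season 15 Battle Pass quest.
; On activation it listens repeatedly for gameplay effects applied to player-tagged targets from a harpoon weapon source.
; Each target pawn not already in the HitOpponents array is added and sends an Athena.Quests.HarpoonHit stat event to the owner's quest manager.
; Repeat hits only print a debug message.
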(event Ability|EventActivateAbility
  ;[1 1 1] Wait for the player to hit an opponent with the harpoon gun
  (|CallParentFunction)
  (bind (_asynctaskproxy _target) (Ability|Tasks|WaitGameplayEffectAppliedToTarget :SourceTagRequirements (Variables|Default|GetTC_HarpoonWeaponSourceTag) :TargetTagRequirements (Variables|Default|GetTC_PlayerTag) :TriggerOnce false :ListenForPeriodicEffects false)
    (:OnApplied
      (CallFunction|HarpoonHit _target))))

(event Ability|EventFailedToActivatePassiveAbility
  (params (ActorInfo :type Gameplay%20Ability%20Actor%20Info%20Structure)))

(event Custom|HarpoonHit
  (params (TargetActor :type Actor%20Object%20Reference%20%28by%20ref%29))
  ;[1 1 1] Check if target has already been hit by owning player; if not, add them to the array and send stat event
  (bind _asfort_player_pawn (Utilities|Casting|CastToFortPlayerPawn TargetActor))
  (if (Utilities|Array|ContainsItem (Variables|Default|GetHitOpponents) _asfort_player_pawn)
    (Development|PrintString "Already hit player" true true "(R=1.000000,G=0.000000,B=0.210322,A=1.000000)")
    (else
      (Utilities|Array|Add (Variables|Default|GetHitOpponents) _asfort_player_pawn)
      (Quest|SendComplexCustomStatEvent (Variables|GAVars|Data|GetOwningPlayerQuestManager) 0 "(GameplayTags=((TagName=\"Athena.Quests.HarpoonHit\")))" "(GameplayTags=)" 1))))
